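; Blueprint actor that drives cube-movement corruption spread on the landscape from calendar events, with a separate front-end path.
; Editor functions stamp cube path footsteps in reverse order into a time gradient, optionally encrypted, and export it as a texture.
; ProcessStaticTexture thresholds that gradient by corruption time into a render target, sets material collection parameters and flushes virtual textures.
; Clients build a coverage map from the result to toggle ambient corruption audio.
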(event EventBeginPlay
  (bind _returnvalue (Game|GetPlayerController 0))
  (Variables|CorruptionCoverageMap|Setinternal_bitArrayRT "")
  (Variables|Generation|SetInternalRenderTargetFootstep "")
  (Variables|Generation|SetInternalRenderTargetLayered "")
  (Variables|Generation|SetInternalRenderTargetLayeredCopy "")
  (Variables|Default|SetRenderTarget_thresholded "")
  (Development|PrintString "BP_CubeMovementGradient BeginPlay")
  (Utilities|IsValid _returnvalue
    (:Is%20Valid
      ;[1 1 1] in FE?
      (Utilities|Casting|CastToFortPlayerStateFrontEnd (Variables|/Script/Engine.Controller|GetPlayerState _returnvalue)
        (:then
          ;[1 1 1] FE Calendatr Events
          (Development|PrintString "BP_CubeMovementGradient FrontEndContext")
          (MCPContext|InitialCalendarSyncCompleted
            (:OnSyncComplete
              (bind (_biseventactive _timesincebegin _timeuntilend _timespanratio) (CalendarEvents|GetCalendarEventInformation "LCCS01"))
              (if _biseventactive
                ;;[1 1 1] First Wave
                (Development|PrintString (Utilities|String|Append "Event LCCS01 Active " (Utilities|String|ToString(Float) _timespanratio)))
                (CallFunction|ProcessStaticTextureFE 0.0 0.0 (Utilities|Struct|MakeCorruptionSpreadStruct _timespanratio))
                (else
                  (bind _biseventactive_2 (CalendarEvents|GetCalendarEventInformation "LCCSP01"))
                  (if _biseventactive_2
                    ;;[1 1 1] Pause
                    (Development|PrintString "Event LCCSP01 Active")
                    (CallFunction|ProcessStaticTextureFE 1.0 0.0 (Utilities|Struct|MakeCorruptionSpreadStruct 1.0))
                    (else
                      (bind (_biseventactive_3 _timesincebegin_2 _timeuntilend_2 _timespanratio_2) (CalendarEvents|GetCalendarEventInformation "LCCS02"))
                      (if _biseventactive_3
                        ;;[1 1 1] CubeTownCorruptionSpread is active - this is the second wave of corruption
                        (Development|PrintString "Event LCCS02 Active 1.0")
                        (CallFunction|ProcessStaticTextureFE 1.0 0.0 (Utilities|Struct|MakeCorruptionSpreadStruct 1.0 _timespanratio_2))
                        (else
                          ;;[1 1 1] events aren't active .. 0 corruption.
                          (Development|PrintString "No LCC Events Active 0.001")
                          (CallFunction|ProcessStaticTextureFE 0.001 0.0 (Utilities|Struct|MakeCorruptionSpreadStruct 0.001 0.001)))))))))))
        (:CastFailed
          _)))
    (:Is%20Not%20Valid
      _))
  ;[1 1 1] Gameplay Calendar
  (Development|PrintString "BP_CubeMovementGradient NOT FrontEndContext")
  (Class|FortAsyncActionAuthorityMatchReady|MatchReadyAsync
    (:OnReady
      (Development|PrintString "BP_CubeMovementGradient MatchReady")
      (Utilities|FlowControl|Sequence
        (:then_0
          (bind (_biseventactive_4 _timesincebegin_3 _timeuntilend_3 _timespanratio_3) (CalendarEvents|GetCalendarEventInformation "CSCM01")) ; entire duration of move1 == corruption duration\ndo not use CCS01, talk with tha; we're deleting that
          (if _biseventactive_4
            ;;[1 1 1] CSCM01 is active - this is the first wave of corruption.
            (Development|PrintString (Utilities|String|Append "CSCM01 Active " (Utilities|String|ToString(Float) _timespanratio_3)))
            (Variables|Default|SetcorruptionSpread (Utilities|Struct|MakeCorruptionSpreadStruct _timespanratio_3))
            (else
              (bind _biseventactive_5 (CalendarEvents|GetCalendarEventInformation "CCSP01"))
              (if _biseventactive_5
                ;;[1 1 1] CubeCorruptionSpreadPaused
                (Development|PrintString (Utilities|String|Append "CCSP01 Active "))
                (Variables|Default|SetcorruptionSpread (Utilities|Struct|MakeCorruptionSpreadStruct 1.0))
                (else
                  ;;[1 1 1] CTC01 : CubeTownCorruptionSpread
                  (bind (_biseventactive_6 _timesincebegin_4 _timeuntilend_4 _timespanratio_4) (CalendarEvents|GetCalendarEventInformation "CTCS01"))
                  (if _biseventactive_6
                    ;;[1 1 1] CubeTownCorruptionSpread is active - this is the second wave of corruption
                    (Development|PrintString (Utilities|String|Append "CTCS01 Active " (Utilities|String|ToString(Float) _timespanratio_4)))
                    (Variables|Default|SetcorruptionSpread (Utilities|Struct|MakeCorruptionSpreadStruct 1.0 _timespanratio_4))
                    (else
                      ;;[1 1 1] events aren't active .. 0 corruption.
                      (Development|PrintString "No Events Active 0.00001")
                      (Variables|Default|SetcorruptionSpread (Utilities|Struct|MakeCorruptionSpreadStruct 0.001 0.001)))))))))
        (:then_1
          (if false
            (CallFunction|DebugStartOffsets)))))))

(event Custom|AsyncLoadLandscape
  (bind _object (Utilities|AsyncLoadAsset (Variables|Landscape|Getlandscape))
    (:Completed
      (bind _aslandscape (Utilities|Casting|CastToLandscape _object)
        (:then
          _)
        (:CastFailed
          _))
      (Default|SetupLandscapeBounds(Landscape) _aslandscape)
      (Default|ProcessStaticTexture (Variables|Default|GetcorruptionSpread))
      (Debug|Flush))))

(event EventEndPlay
  (params (EndPlayReason :type EEndPlayReason%20Enum :default "Destroyed"))
  (switch Utilities|FlowControl|Switch|SwitchonEEndPlayReason EndPlayReason
    (:Destroyed
      _)
    (:LevelTransition
      _)
    (:RemovedFromWorld
      _)
    (:Quit
      _))
  (|CallParentFunction EndPlayReason)
  (Networking|CanExecuteCosmeticEvents
    (:True
      (Development|PrintString "EndPlay")
      (Rendering|ClearRenderTarget2D (Variables|Default|GetRenderTarget_processed)))))

(event EventDestroyed
  (|CallParentFunction)
  (Networking|CanExecuteCosmeticEvents
    (:True
      (Development|PrintString "Destroyed")
      (Rendering|ClearRenderTarget2D (Variables|Default|GetRenderTarget_processed)))))

(event Custom|CE_InitializeCorruptionCoverageMap
  (Networking|SwitchHasAuthority
    (:Remote
      ;[1 1 1] NOTE:  This node ends up calling InitializeCorruptionCoverageMap
      (CorruptionCoverageMap|GetBitArray)
      (Utilities|Time|SetTimerbyEvent (delegate Custom|CustomEvent) (Variables|CorruptionCoverageMap|GetCorruptionStateUpdatePeriod) true)
      (CorruptionCoverageMap|BindEventtoOnIsAnyLocalPlayerinCorruptionChanged (EventDispatchers|CreateEvent :!event Custom|HandleCorruptionStateChanged)))))

(event Custom|CustomEvent
  (CorruptionCoverageMap|UpdateCorruptionState))

(event Custom|HandleCorruptionStateChanged
  (params (Is%20In%20Corruption :type Boolean))
  (bind _returnvalue (WorldSubsystems|GetAmbientAudioSubsystem))
  (bind _ambientaudioentryname (Variables|AmbientAudio|GetAmbientAudioEntryName))
  (bind _ambientaudiosoundmix (Variables|AmbientAudio|GetAmbientAudioSoundMix))
  ;[1 1 1] Respond to changes in Corruption state (audio)
  (Networking|SwitchHasAuthority
    (:Remote
      (if Is%20In%20Corruption
        (AmbientAudio|AddAmbientEntry _returnvalue _ambientaudioentryname (Variables|AmbientAudio|GetCorruptionAmbientAudioBank) -1 5.0)
        (Audio|PushSoundMixModifier _ambientaudiosoundmix)
        (else
          (AmbientAudio|RemoveAmbientEntry _returnvalue _ambientaudioentryname -1.0)
          (Audio|PopSoundMixModifier _ambientaudiosoundmix))))))

(fn A%29SetupSplines
  (Development|PrintString "SETUP")
  (Default|FillCubeMovrmentStaticPathsArray)
  (CallFunction|FillCubeMovementArray)
  (Default|SortFootStepsUsingPython))

(fn b%29GenerateRTs
  (bind _times (Variables|Hide|Gettimes))
  (Utilities|IsValid (Variables|Generation|GetInternalRenderTargetFootstep)
    (:Is%20Valid
      (if (== (Variables|/Script/Engine.TextureRenderTarget2D|GetSizeX (Variables|Generation|GetInternalRenderTargetFootstep)) (Variables|Generation|GetRenderTargetResolution))
        ;[1 1 1] clear render targets if they already exist
        (Rendering|ClearRenderTarget2D (Variables|Generation|GetInternalRenderTargetFootstep) "(R=0.000000,G=0.000000,B=0.000000,A=0.000000)")
        (Rendering|ClearRenderTarget2D (Variables|Generation|GetInternalRenderTargetLayered) "(R=0.000000,G=0.000000,B=0.000000,A=0.000000)")
        (Rendering|ClearRenderTarget2D (Variables|Generation|GetInternalRenderTargetLayeredCopy) "(R=0.000000,G=0.000000,B=0.000000,A=0.000000)")
        (else
          ;[1 1 1] Create a RenderTarget2d
          (bind _returnvalue (Rendering|CreateRenderTarget2D (Variables|Generation|GetRenderTargetResolution) (Variables|Generation|GetRenderTargetResolution) "RTF_RGBA16f" "(R=0.000000,G=0.000000,B=0.000000,A=0.000000)"))
          (bind _output_get (Variables|Generation|SetInternalRenderTargetFootstep _returnvalue))
          (Variables|/Script/Engine.Texture|SetFilter "TF_Nearest" _output_get)
          (Variables|/Script/Engine.Texture|SetLODGroup "TEXTUREGROUP_Pixels2D" _output_get)
          (bind _returnvalue_2 (Rendering|CreateRenderTarget2D (Variables|Generation|GetRenderTargetResolution) (Variables|Generation|GetRenderTargetResolution) "RTF_RGBA16f" "(R=0.000000,G=0.000000,B=0.000000,A=0.000000)"))
          (bind _output_get_2 (Variables|Generation|SetInternalRenderTargetLayered _returnvalue_2))
          (Variables|/Script/Engine.Texture|SetFilter "TF_Nearest" _output_get_2)
          (Variables|/Script/Engine.Texture|SetLODGroup "TEXTUREGROUP_Pixels2D" _output_get_2)
          (bind _returnvalue_3 (Rendering|CreateRenderTarget2D (Variables|Generation|GetRenderTargetResolution) (Variables|Generation|GetRenderTargetResolution) "RTF_RGBA16f" "(R=0.000000,G=0.000000,B=0.000000,A=0.000000)"))
          (bind _output_get_3 (Variables|Generation|SetInternalRenderTargetLayeredCopy _returnvalue_3))
          (Variables|/Script/Engine.Texture|SetFilter "TF_Nearest" _output_get_3)
          (Variables|/Script/Engine.Texture|SetLODGroup "TEXTUREGROUP_Pixels2D" _output_get_3)))
      ;[1 1 1] setup DMIs
      (bind _returnvalue_4 (Rendering|Material|CreateDynamicMaterialInstance "/CorruptionGameplay/Materials/M_corruptionGradient_combine.M_corruptionGradient_combine" "None" "None" :!class /Script/Engine.KismetMaterialLibrary))
      (bind _output_get_4 (LocalVariables|SetDMI_combine _returnvalue_4))
      (Rendering|Material|SetTextureParameterValue _output_get_4 "FootstepTexture" (Variables|Generation|GetInternalRenderTargetFootstep))
      (Rendering|Material|SetTextureParameterValue _output_get_4 "LayeredTexture" (Variables|Generation|GetInternalRenderTargetLayeredCopy))
      (bind _returnvalue_5 (Rendering|Material|CreateDynamicMaterialInstance "/CorruptionGameplay/Materials/M_corruptionGradient_copy.M_corruptionGradient_copy" "None" "None" :!class /Script/Engine.KismetMaterialLibrary))
      (bind _output_get_5 (LocalVariables|SetDMI_copy _returnvalue_5))
      (Rendering|Material|SetTextureParameterValue _output_get_5 "Texture" (Variables|Generation|GetInternalRenderTargetLayered))
      ;[1 1 1] work backwards throught the position list : the earliest footprints need to be layered above later ones.
      (bind (_arrayindex _arrayelement) (Utilities|Array|ReverseforEachLoop (Variables|Hide|Getpositions))
        (:LoopBody
          (Rendering|ClearRenderTarget2D (Variables|Generation|GetInternalRenderTargetFootstep) "(R=0.000000,G=0.000000,B=0.000000,A=0.000000)")
          ;[1 1 1] draw the footprint
          (bind _returnvalue_6 (CallFunction|PostoTextureCoord _arrayelement))
          (Default|DrawCubeToRT 0 (Math|Float|Truncate (.x _returnvalue_6)) (Math|Float|Truncate (.y _returnvalue_6)) "(R=1.000000,G=1.000000,B=1.000000,A=1.000000)" (* (Math|Random|RandomFloat) 3.6e+02) (* (* (/ (Math|Conversions|ToFloat(Integer) (Variables|Generation|GetRenderTargetResolution)) 1024.0) (Math|Interpolation|Ease "Linear" (Utilities|Array|Get(acopy) _times _arrayindex) (+ (Variables|Generation|GetcubeScaleMult) (* (* (Variables|Generation|GetcubeScaleMult) 0.2) (Math|Trig|Sin(Degrees) (* (Utilities|Array|Get(acopy) _times _arrayindex) 2e+03)))) (* (+ (Variables|Generation|GetcubeScaleMult) (* (* (Variables|Generation|GetcubeScaleMult) 0.2) (Math|Trig|Sin(Degrees) (* (Utilities|Array|Get(acopy) _times _arrayindex) 2e+03)))) 0.5))) (Utilities|Array|Get(acopy) (Variables|Hide|Getscales) _arrayindex)) (Variables|Generation|GetInternalRenderTargetFootstep))
          (Rendering|Material|SetScalarParameterValue (LocalVariables|GetDMI_combine) "CorruptionTime" (Utilities|Array|Get(acopy) _times _arrayindex) :!class /Script/Engine.MaterialInstanceDynamic)
          (Rendering|DrawMaterialtoRenderTarget (Variables|Generation|GetInternalRenderTargetLayered) (LocalVariables|GetDMI_combine))
          (Rendering|DrawMaterialtoRenderTarget (Variables|Generation|GetInternalRenderTargetLayeredCopy) (LocalVariables|GetDMI_copy)))
        (:Completed
          ;[1 1 1] encrypt
          (if (Variables|Generation|GetencryptGradient)
            (bind _returnvalue_7 (Rendering|Material|CreateDynamicMaterialInstance "/CorruptionGameplay/Materials/M_corruptionGradient_encrypt.M_corruptionGradient_encrypt" "None" "None" :!class /Script/Engine.KismetMaterialLibrary))
            (bind _output_get_6 (LocalVariables|SetDMI_encrypt _returnvalue_7))
            (Rendering|Material|SetTextureParameterValue _output_get_6 "Texture" (Variables|Generation|GetInternalRenderTargetLayeredCopy))
            (Rendering|DrawMaterialtoRenderTarget (Variables|Generation|GetInternalRenderTargetLayered) _output_get_6)
            (LocalVariables|SetDMI_encrypt 0)))))
    (:Is%20Not%20Valid
      _)))

(fn C%29ExportRTstoStaticTexture
  (bind _gradienttexture (Variables|Default|GetgradientTexture))
  (bind _returnvalue (Utilities|GetDisplayName _gradienttexture))
  (bind _returnvalue_2 (Utilities|Paths|GetSystemPath _gradienttexture))
  (bind _returnvalue_3 (Utilities|Paths|GetPath _returnvalue_2))
  (bind _returnvalue_4 (Utilities|String|Append _returnvalue_3 "/"))
  (bind _returnvalue_5 (Utilities|String|Append _returnvalue ".exr"))
  (bind _returnvalue_6 (Utilities|String|Append _returnvalue_4 _returnvalue_5))
  (bind _gradienttexture_2 (Variables|Default|GetgradientTexture))
  ;[1 1 1] export
  (Development|PrintString _returnvalue_6)
  (Rendering|ExportRenderTarget (Variables|Generation|GetInternalRenderTargetLayered) _returnvalue_4 _returnvalue_5)
  (Utilities|FlowControl|Sequence
    (:then_0
      (bind _returnvalue_7 (Game|ConstructObjectfromClass "/Script/UnrealEd.AutomatedAssetImportData" self)) ; Construct From Object Class Node
      (bind _output_get (LocalVariables|SetImportData _returnvalue_7))
      (Variables|/Script/UnrealEd.AutomatedAssetImportData|SetFilenames (Utilities|Array|MakeArray _returnvalue_6) _output_get)
      (Variables|/Script/UnrealEd.AutomatedAssetImportData|SetDestinationPath (Utilities|Paths|GetPath (Utilities|GetObjectPathString _gradienttexture)) _output_get)
      (Variables|/Script/UnrealEd.AutomatedAssetImportData|SetbReplaceExisting true _output_get))
    (:then_1
      (bind _returnvalue_8 (EditorScripting|AssetTools|GetAssetTools))
      (bind _returnvalue_9 (EditorScripting|AssetTools|ImportAssetsAutomated _returnvalue_8 _returnvalue_7))
      (for _array_element _returnvalue_9
        (Development|PrintString (Utilities|GetObjectPathString _array_element))
        (Variables|/Script/Engine.Texture|SetCompressionSettings "TC_HalfFloat" _gradienttexture_2)
        (Variables|/Script/Engine.Texture|SetFilter "TF_Nearest" _gradienttexture_2)
        (Variables|/Script/Engine.Texture|SetMipGenSettings "TMGS_NoMipmaps" _gradienttexture_2)))))

(fn clearfootsteps
  (Development|PrintString (Utilities|String|Append "num positions " (Utilities|String|ToString(Integer) (Utilities|Array|Length (Variables|Hide|Getpositions)))))
  (Utilities|Array|Clear (Variables|Hide|Getscales))
  (Utilities|Array|Clear (Variables|Hide|Getpositions))
  (Utilities|Array|Clear (Variables|Hide|Gettimes))
  (Development|PrintString (Utilities|String|Append "num positions " (Utilities|String|ToString(Integer) (Utilities|Array|Length (Variables|Hide|Getpositions))))))

(fn clearRT
  (Rendering|ClearRenderTarget2D (Variables|Default|GetRenderTarget_processed))
  (Rendering|ClearRenderTarget2D (Variables|Generation|GetInternalRenderTargetFootstep) "(R=0.000000,G=0.000000,B=0.000000,A=0.000000)")
  (Rendering|ClearRenderTarget2D (Variables|Generation|GetInternalRenderTargetLayered) "(R=0.000000,G=0.000000,B=0.000000,A=0.000000)"))

(fn d%29ProcessStaticTexture
  (Default|ProcessStaticTexture (Variables|Default|GetcorruptionSpread)))

(fn DebugStartOffsetforEvent
  (params (CalendarEvent :type String) (CorruptionStartEvent :type String))
  (bind (_biseventactive _timesincebegin) (CalendarEvents|GetCalendarEventInformation CalendarEvent))
  (if _biseventactive
    (LocalVariables|SetTimeSinceBegin_Event _timesincebegin)
    (bind (_biseventactive_2 _timesincebegin_2 _timeuntilend) (CalendarEvents|GetCalendarEventInformation CorruptionStartEvent)) ; Get how long it has been since corruption started.\n
    (if _biseventactive_2
      (LocalVariables|SetTimeSinceCorruptionBegun _timesincebegin_2)
      (LocalVariables|SetTimeUntilCorruptionEnds _timeuntilend)
      (Development|PrintString (Utilities|String|Append (Variables|GetCalendarEvent) " begins approximately at corruption %: " (Utilities|String|ToString(Float) (Math|Timespan|TimespanRatio (Math|Timespan|Timespan-Timespan (LocalVariables|GetTimeSinceCorruptionBegun) (LocalVariables|GetTimeSinceBegin_Event)) (Math|Timespan|Timespan+Timespan (LocalVariables|GetTimeUntilCorruptionEnds) (LocalVariables|GetTimeSinceCorruptionBegun))))) true true "(R=1.000000,G=0.737268,B=0.000000,A=1.000000)" 1e+01)
      (else
        (Development|PrintString (Utilities|String|Append "Corruption Event Not Started. EventName:" CorruptionStartEvent) true true "(R=1.000000,G=0.005426,B=0.000000,A=1.000000)" 1e+01)))
    (else
      (Development|PrintString (Utilities|String|Append "Calendar Event Not Active, Please use a later calendar offset" CalendarEvent) true true "(R=1.000000,G=0.000000,B=0.019224,A=1.000000)" 1e+01))))

(fn DebugStartOffsets
  (CallFunction|DebugStartOffsetForEvent "CSCM02" "CSCM01") ; CSCM01 is essentially corruption,\nwe're going to remove the other event. (CCS01)\nUse what is printed here to update ApolloCubeMovementCorruptionSequence
  (CallFunction|DebugStartOffsetForEvent "CSCM03" "CSCM01")
  (CallFunction|DebugStartOffsetForEvent "CSCM04" "CSCM01")
  (CallFunction|DebugStartOffsetForEvent "CSCM05" "CSCM01"))

(fn DrawCubetoRT
  (params (canvas :type Canvas%20Object%20Reference) (x :type Integer) (y :type Integer) (color :type Linear%20Color%20Structure) (rotation :type Float%20%28double-precision%29) (size :type Float%20%28double-precision%29) (renderTarget :type Texture%20Render%20Target%202D%20Object%20Reference))
  (bind (_canvas _size _context) (Rendering|BeginDrawCanvastoRenderTarget renderTarget))
  (Canvas|DrawTexture _canvas "/CorruptionGameplay/TechArt/splat.splat" (Math|Vector2D|MakeVector2D (- (Math|Conversions|ToFloat(Integer) x) (* size 0.5)) (- (Math|Conversions|ToFloat(Integer) y) (* size 0.5))) (Math|Vector2D|MakeVector2D size size) "(X=0.750000,Y=0.000000)" "(X=0.250000,Y=0.250000)" color "BLEND_Opaque" rotation "(X=0.500000,Y=0.500000)")
  (Rendering|EndDrawCanvastoRenderTarget _context))

(fn DrawCubetoRT_old
  (params (canvas :type Canvas%20Object%20Reference) (x :type Integer) (y :type Integer) (color :type Linear%20Color%20Structure) (rotation :type Float%20%28double-precision%29) (size :type Float%20%28double-precision%29) (renderTarget :type Texture%20Render%20Target%202D%20Object%20Reference))
  (bind (_canvas _size _context) (Rendering|BeginDrawCanvastoRenderTarget renderTarget))
  (Canvas|DrawTexture _canvas "" (Math|Vector2D|MakeVector2D (- (Math|Conversions|ToFloat(Integer) x) (* size 0.5)) (- (Math|Conversions|ToFloat(Integer) y) (* size 0.5))) (Math|Vector2D|MakeVector2D size size) "(X=0.000000,Y=0.000000)" "(X=1.000000,Y=1.000000)" color "BLEND_Opaque" rotation "(X=0.500000,Y=0.500000)")
  (Rendering|EndDrawCanvastoRenderTarget _context))

(fn E%29<<
  (bind (_corruptionspread_cubecorruptionspread_7_62751f0f42c1cecd5f6b69b8e7717293 _corruptionspread_cubetowncorruptionspread_4_143c45964a7b105562cb508ad98ce6fb) (Utilities|Struct|BreakCorruptionSpreadStruct (Variables|GetcorruptionSpread_CubeCorruptionSpread_7_62751F0F42C1CECD5F6B69B8E7717293)))
  (if (> _corruptionspread_cubetowncorruptionspread_4_143c45964a7b105562cb508ad98ce6fb 0.0)
    (bind _output_get (Variables|SetcorruptionSpread_CubeCorruptionSpread_7_62751F0F42C1CECD5F6B69B8E7717293 (Utilities|Struct|MakeCorruptionSpreadStruct 1.0 (Math|Float|Clamp(Float) (- _corruptionspread_cubetowncorruptionspread_4_143c45964a7b105562cb508ad98ce6fb (Variables|Default|GetPreviewTimeStep))))))
    (LocalVariables|Setspread _output_get)
    (else
      (bind _output_get_2 (Variables|SetcorruptionSpread_CubeCorruptionSpread_7_62751F0F42C1CECD5F6B69B8E7717293 (Utilities|Struct|MakeCorruptionSpreadStruct (Math|Float|Clamp(Float) (- _corruptionspread_cubecorruptionspread_7_62751f0f42c1cecd5f6b69b8e7717293 (Variables|Default|GetPreviewTimeStep))))))
      (LocalVariables|Setspread _output_get_2)))
  (Default|ProcessStaticTexture (LocalVariables|Getspread))
  (Debug|Flush)
  (if (Variables|Default|GetpreviewCubes)
    (CallFunction|PreviewCubeInstances)
    (else
      (Default|PreviewCubeInstancesClear))))

(fn f%29>>
  (bind (_corruptionspread_cubecorruptionspread_7_62751f0f42c1cecd5f6b69b8e7717293 _corruptionspread_cubetowncorruptionspread_4_143c45964a7b105562cb508ad98ce6fb) (Utilities|Struct|BreakCorruptionSpreadStruct (Variables|GetcorruptionSpread_CubeCorruptionSpread_7_62751F0F42C1CECD5F6B69B8E7717293)))
  (if (>= _corruptionspread_cubecorruptionspread_7_62751f0f42c1cecd5f6b69b8e7717293 1.0)
    (bind _output_get (Variables|SetcorruptionSpread_CubeCorruptionSpread_7_62751F0F42C1CECD5F6B69B8E7717293 (Utilities|Struct|MakeCorruptionSpreadStruct 1.0 (Math|Float|Clamp(Float) (+ _corruptionspread_cubetowncorruptionspread_4_143c45964a7b105562cb508ad98ce6fb (Variables|Default|GetPreviewTimeStep))))))
    (LocalVariables|Setspread _output_get)
    (else
      (bind _output_get_2 (Variables|SetcorruptionSpread_CubeCorruptionSpread_7_62751F0F42C1CECD5F6B69B8E7717293 (Utilities|Struct|MakeCorruptionSpreadStruct (Math|Float|Clamp(Float) (+ _corruptionspread_cubecorruptionspread_7_62751f0f42c1cecd5f6b69b8e7717293 (Variables|Default|GetPreviewTimeStep))))))
      (LocalVariables|Setspread _output_get_2)))
  (Default|ProcessStaticTexture (LocalVariables|Getspread))
  (Debug|Flush)
  (if (Variables|Default|GetpreviewCubes)
    (CallFunction|PreviewCubeInstances)
    (else
      (Default|PreviewCubeInstancesClear))))

(fn FillCubeMovementArray
  (Utilities|Array|Clear (Variables|Hide|Getscales))
  (Utilities|Array|Clear (Variables|Hide|Getpositions))
  (Utilities|Array|Clear (Variables|Hide|Gettimes))
  ;[1 1 1] loop through static paths
  (for (_array_element _array_index) (Variables|Generation|GetcubeMovementStaticPaths)
    (Utilities|IsValid (Utilities|ResolveSoftReference _array_element)
      (:Is%20Valid
        (Development|PrintString (Utilities|String|Append "StaticPath " (Utilities|GetDisplayName (Utilities|ResolveSoftReference _array_element))))
        ;[1 1 1] loop through all travellers
        (bind _output_get (LocalVariables|SetNewLocalVar_0 (Utilities|Struct|BreakCubeMovementCorruptionGenerationData (Class|CubeMovementStaticPath|EditorGetCorruptionGenerationData (Utilities|ResolveSoftReference _array_element)))))
        (for _array_element_2 _output_get
          (Development|PrintString (Utilities|String|Append "traveler " (Utilities|GetDisplayName (.PathTraveler (Utilities|Struct|BreakCubeMovementCorruptionGenerationTravelerData _array_element_2))) " " (Utilities|String|ToString(Integer) (Utilities|Array|Length (Utilities|Struct|BreakCubeMovementCorruptionGenerationTravelerData _array_element_2)))))
          (for _array_element_3 (Utilities|Struct|BreakCubeMovementCorruptionGenerationTravelerData _array_element_2)
            (bind _returnvalue_x (Math|Vector|BreakVector (Transformation|GetActorScale3D (.PathTraveler (Utilities|Struct|BreakCubeMovementCorruptionGenerationTravelerData _array_element_2)))))
            (Utilities|Array|Add (Variables|Hide|Getscales) _returnvalue_x)
            (Utilities|Array|Add (Variables|Hide|Getpositions) (.location (Utilities|Struct|BreakCubeMovementCorruptionGenerationSplinePointData _array_element_3)))
            (Utilities|Array|Add (Variables|Hide|Gettimes) (.SplinePercentComplete (Utilities|Struct|BreakCubeMovementCorruptionGenerationSplinePointData _array_element_3))))))))
  (Development|PrintString (Utilities|String|Append "num positions " (Utilities|String|ToString(Integer) (Utilities|Array|Length (Variables|Hide|Getpositions))))))

(fn FillCubeMovrmentStaticPathsArray
  (Utilities|Array|Clear (Variables|Generation|GetcubeMovementStaticPaths))
  (bind _outactors (Actor|GetAllActorsOfClass "/Script/CorruptionGameplayCodeRuntime.CubeMovementStaticPath"))
  (for _array_element _outactors
    (if (not (or (Utilities|String|Contains (Utilities|GetDisplayName _array_element) "gold") (Utilities|String|Contains (Utilities|GetDisplayName _array_element) "Stomp")))
      (Utilities|Array|Add (Variables|Generation|GetcubeMovementStaticPaths) (Utilities|ResolveSoftReference _array_element)))))

(fn flush
  (Development|ExecuteConsoleCommand "r.vt.Flush 1"))

(fn GetBitArray
  (Development|PrintString "Get Bit Array")
  ;[1 1 1] Create an RT for the BitArray process to read from (has to be rgba8 for the read pixel code to work)
  (bind _returnvalue (Rendering|CreateRenderTarget2D (Variables|Generation|GetRenderTargetResolution) (Variables|Generation|GetRenderTargetResolution) "RTF_RGBA8" "(R=0.000000,G=0.000000,B=0.000000,A=0.000000)"))
  (bind _output_get (Variables|CorruptionCoverageMap|Setinternal_bitArrayRT _returnvalue))
  (Variables|/Script/Engine.Texture|SetFilter "TF_Nearest" _output_get)
  (Variables|/Script/Engine.Texture|SetLODGroup "TEXTUREGROUP_Pixels2D" _output_get)
  ;[1 1 1] setup bitArrayRT to read from - this is a rgba8 copy of renderTargetProcessed as the RT reader only seems to work on rgba8 and not r8
  (Rendering|ClearRenderTarget2D (Variables|CorruptionCoverageMap|Getinternal_bitArrayRT))
  (bind _returnvalue_2 (Rendering|Material|CreateDynamicMaterialInstance "/CorruptionGameplay/Materials/M_corruptionGradient_bitArray.M_corruptionGradient_bitArray" "None" "None" :!class /Script/Engine.KismetMaterialLibrary))
  (Rendering|Material|SetTextureParameterValue _returnvalue_2 "Corruption" (Variables|Default|GetRenderTarget_processed))
  (Rendering|DrawMaterialtoRenderTarget (Variables|CorruptionCoverageMap|Getinternal_bitArrayRT) _returnvalue_2)
  (Utilities|IsValid (Variables|CorruptionCoverageMap|Getinternal_bitArrayRT)
    (:Is%20Valid
      (CorruptionCoverageMap|InitializeCorruptionCoverageMap (Variables|CorruptionCoverageMap|Getinternal_bitArrayRT))
      ;[1 1 1] cleanup BitArrayRT
      (if (Variables|Default|Getcleanup)
        (Variables|CorruptionCoverageMap|Setinternal_bitArrayRT "")))))

(fn GetCubeTransformsAtTime
  (params (time :type Float%20%28double-precision%29))
  (LocalVariables|SetlocalTime time)
  (Utilities|Array|Clear (LocalVariables|Gettransformss))
  (for _array_element (Variables|Generation|GetcubeMovementStaticPaths)
    (bind _output_get (LocalVariables|SetNewLocalVar_1 (Utilities|Struct|BreakCubeMovementCorruptionGenerationData (Class|CubeMovementStaticPath|EditorGetCorruptionGenerationData (Utilities|ResolveSoftReference _array_element)))))
    (for _array_element_2 _output_get
      (bind _array_element_3 (Utilities|Array|ForEachLoopwithBreak (Utilities|Struct|BreakCubeMovementCorruptionGenerationTravelerData _array_element_2))
        (:LoopBody
          (if (<= (.SplinePercentComplete (Utilities|Struct|BreakCubeMovementCorruptionGenerationSplinePointData _array_element_3)) (LocalVariables|GetlocalTime))
            _
            (else
              _))
          (Variables|Settransform_Location (Math|Transform|MakeTransform (.location (Utilities|Struct|BreakCubeMovementCorruptionGenerationSplinePointData _array_element_3)) (.rotation (Utilities|Struct|BreakCubeMovementCorruptionGenerationSplinePointData _array_element_3)) (Transformation|GetActorScale3D (.PathTraveler (Utilities|Struct|BreakCubeMovementCorruptionGenerationTravelerData _array_element_2))))))
        (:Completed
          _))
      (Utilities|Array|Add (LocalVariables|Gettransformss) (LocalVariables|Gettransform))))
  (return (LocalVariables|Gettransformss)))

(fn InitializeCorruptionCoverageMap
  (params (CorruptionRT :type Texture%20Render%20Target%202D%20Object%20Reference))
  (bind _returnvalue (Game|ConstructObjectfromClass "/Script/CorruptionGameplayCodeRuntime.CorruptionCoverageMap" self))
  (bind _output_get (Variables|CorruptionCoverageMap|SetCorruptionCoverageMap _returnvalue))
  (bind _returnvalue_2 (CorruptionCoverageMap|UpdateCorruptionCoverageMap _output_get CorruptionRT (* (Variables|Landscape|GetLandscapeBias) -1.0) (+ (* (Variables|Landscape|GetLandscapeBias) -1.0) (Variables|Landscape|GetLandscapeSize)) 0.1 (Math|Float|SelectFloat 1e+01 -1.0 (Variables|CorruptionCoverageMap|GetDebugCorruptionCoverageMap))))
  (return _returnvalue_2))

(fn OnRep_corruptionTime
  (Development|PrintString "On Rep Corruption Time")
  (Networking|SwitchHasAuthority
    (:Remote
      (CallFunction|AsyncLoadLandscape))))

(fn PosToTextureCoord
  (params (Location :type Vector))
  (return (* (/ (+ Location (Variables|Landscape|GetLandscapeBias)) (Variables|Landscape|GetLandscapeSize)) (Variables|Generation|GetRenderTargetResolution))))

(fn PreviewCubeInstances
  (bind _instancedstaticmesh (Variables|Default|GetInstancedStaticMesh))
  (bind _corruptionspread_cubecorruptionspread_7_62751f0f42c1cecd5f6b69b8e7717293 (Utilities|Struct|BreakCorruptionSpreadStruct (Variables|GetcorruptionSpread_CubeCorruptionSpread_7_62751F0F42C1CECD5F6B69B8E7717293)))
  (if (Variables|Default|GetpreviewCubes)
    (bind _transforms (CallFunction|GetCubeTransformsatTime _corruptionspread_cubecorruptionspread_7_62751f0f42c1cecd5f6b69b8e7717293))
    (Components|InstancedStaticMesh|ClearInstances _instancedstaticmesh)
    (for _array_element _transforms
      (bind (_array_element_location _array_element_rotation _array_element_scale) (Math|Transform|BreakTransform _array_element))
      (Components|InstancedStaticMesh|AddInstance _instancedstaticmesh (Math|Transform|MakeTransform (Math|Vector|vector+vector _array_element_location (Math|Vector|MakeVector 0.0 0.0 (* (.x _array_element_scale) 7e+02))) _array_element_rotation _array_element_scale) false))
    (else
      (Default|PreviewCubeInstancesClear))))

(fn PreviewCubeInstancesClear
  (Development|PrintString "Clear")
  (Components|InstancedStaticMesh|ClearInstances (Variables|Default|GetInstancedStaticMesh)))

(fn preview_allCubes
  (bind _positions (Variables|Hide|Getpositions))
  (bind _scales (Variables|Hide|Getscales))
  (bind _corruptionspread_cubecorruptionspread_7_62751f0f42c1cecd5f6b69b8e7717293 (Utilities|Struct|BreakCorruptionSpreadStruct (Variables|GetcorruptionSpread_CubeCorruptionSpread_7_62751F0F42C1CECD5F6B69B8E7717293)))
  (Development|PrintString (Utilities|String|Append "drawing " (Utilities|String|ToString(Integer) (Utilities|Array|Length _positions)) " cubes"))
  (for (_array_element _array_index) _positions
    (if (<= (Utilities|Array|Get(aref) (Variables|Hide|Gettimes) _array_index) _corruptionspread_cubecorruptionspread_7_62751f0f42c1cecd5f6b69b8e7717293)
      (Rendering|Debug|DrawDebugBox :Center _array_element :Extent (Math|Vector|MakeVector (* (Utilities|Array|Get(aref) _scales _array_index) 4e+02) (* (Utilities|Array|Get(aref) _scales _array_index) 4e+02) (* (Utilities|Array|Get(aref) _scales _array_index) 4e+02)) :LineColor "(R=0.837160,G=0.024608,B=1.000000,A=1.000000)" :Duration 1e+01 :Thickness 2e+01 :DepthPriority "World")
      (else
        (Rendering|Debug|DrawDebugBox :Center _array_element :Extent (Math|Vector|MakeVector (* (Utilities|Array|Get(aref) _scales _array_index) 4e+02) (* (Utilities|Array|Get(aref) _scales _array_index) 4e+02) (* (Utilities|Array|Get(aref) _scales _array_index) 4e+02)) :LineColor "(R=0.300000,G=0.300000,B=0.300000,A=1.000000)" :Duration 1e+01 :Thickness 2e+01 :DepthPriority "World")))))

(fn PrintPositions
  (CallFunction|FillCubeMovementArray)
  (for (_array_element _array_index) (Variables|Hide|Getpositions)
    (Development|PrintString (Utilities|String|Append (Utilities|String|ToString(Integer) _array_index) "  " (Utilities|String|ToString(Vector) _array_element)))))

(fn print_traveller0
  (for _array_element (Variables|Generation|GetcubeMovementStaticPaths)
    (bind _output_get (LocalVariables|SetNewLocalVar_1 (Utilities|Struct|BreakCubeMovementCorruptionGenerationData (Class|CubeMovementStaticPath|EditorGetCorruptionGenerationData (Utilities|ResolveSoftReference _array_element)))))
    (for _array_element_2 (Utilities|Struct|BreakCubeMovementCorruptionGenerationTravelerData (Utilities|Array|Get(acopy) _output_get))
      (bind (_array_element_splinepointtransform _array_element_splinepercentcomplete) (Utilities|Struct|BreakCubeMovementCorruptionGenerationSplinePointData _array_element_2))
      (Development|PrintString (Utilities|String|Append (Utilities|String|ToString(Float) _array_element_splinepercentcomplete) " " (Utilities|String|ToString(Transform) _array_element_splinepointtransform))))))

(fn printCalendarInfo
  (bind (_biseventactive _timesincebegin _timeuntilend _timespanratio) (CalendarEvents|GetCalendarEventInformation "LCCS01"))
  (Development|PrintString (Utilities|String|Append "LCCS01 " (Utilities|String|ToString(Boolean) _biseventactive) " " (Utilities|String|ToString(Float) _timespanratio)))
  (bind (_biseventactive_2 _timesincebegin_2 _timeuntilend_2 _timespanratio_2) (CalendarEvents|GetCalendarEventInformation "LCCSP01 "))
  (Development|PrintString (Utilities|String|Append "LCCSP01" (Utilities|String|ToString(Boolean) _biseventactive_2) " " (Utilities|String|ToString(Float) _timespanratio_2)))
  (bind (_biseventactive_3 _timesincebegin_3 _timeuntilend_3 _timespanratio_3) (CalendarEvents|GetCalendarEventInformation "LCCS02"))
  (Development|PrintString (Utilities|String|Append "LCCS02 " (Utilities|String|ToString(Boolean) _biseventactive_3) "" (Utilities|String|ToString(Float) _timespanratio_3)))
  (bind (_biseventactive_4 _timesincebegin_4 _timeuntilend_4 _timespanratio_4) (CalendarEvents|GetCalendarEventInformation "CTCS01"))
  (Development|PrintString (Utilities|String|Append "CTCS01" (Utilities|String|ToString(Float) _timespanratio_4)))
  (bind (_biseventactive_5 _timesincebegin_5 _timeuntilend_5 _timespanratio_5) (CalendarEvents|GetCalendarEventInformation "CSCM01 ")) ; entire duration of move1 == corruption duration\ndo not use CCS01, talk with tha; we're deleting that
  (Development|PrintString (Utilities|String|Append "CSCM01 " (Utilities|String|ToString(Float) _timespanratio_5))))

(fn ProcessStaticTexture
  (params (CorruptionSpread_CubeCorruptionSpread_7_62751F0F42C1CECD5F6B69B8E7717293 :type Float%20%28single-precision%29) (CorruptionSpread_CubeTownCorruptionSpread_4_143C45964A7B105562CB508AD98CE6FB :type Float%20%28single-precision%29))
  (bind _corruptiontimeoffset (Variables|Default|GetcorruptionTimeOffset))
  (bind _returnvalue (+ CorruptionSpread_CubeCorruptionSpread_7_62751F0F42C1CECD5F6B69B8E7717293 _corruptiontimeoffset))
  (LocalVariables|SetCubeTownCorruptionSpread CorruptionSpread_CubeTownCorruptionSpread_4_143C45964A7B105562CB508AD98CE6FB)
  ;[1 1 1] set MPC
  (Development|PrintString (Utilities|String|Append "Process Static Texture " "CCS01 " (Utilities|String|ToString(Float) CorruptionSpread_CubeCorruptionSpread_7_62751F0F42C1CECD5F6B69B8E7717293) " (+" (Utilities|String|ToString(Float) _corruptiontimeoffset) ") CTCS01 " (Utilities|String|ToString(Float) (LocalVariables|GetCubeTownCorruptionSpread))))
  (Rendering|Material|SetScalarParameterValue "/Game/Packages/Fortress_SharedMaterials/GlobalMaterialParameters/FortniteMaterialParameters.FortniteMaterialParameters" "Corruption1stStage_0-1" _returnvalue :!class /Script/Engine.KismetMaterialLibrary)
  (Rendering|Material|SetScalarParameterValue "/Game/Packages/Fortress_SharedMaterials/GlobalMaterialParameters/FortniteMaterialParameters.FortniteMaterialParameters" "Corruption2ndStage_ColorLerp" (LocalVariables|GetCubeTownCorruptionSpread) :!class /Script/Engine.KismetMaterialLibrary)
  ;[1 1 1] create a threshold RT
  (bind _returnvalue_2 (Rendering|CreateRenderTarget2D (Variables|Generation|GetRenderTargetResolution) (Variables|Generation|GetRenderTargetResolution) "RTF_RGBA16f" "(R=0.000000,G=0.000000,B=0.000000,A=0.000000)"))
  (bind _output_get (Variables|Default|SetRenderTarget_thresholded _returnvalue_2))
  (Variables|/Script/Engine.Texture|SetFilter "TF_Nearest" _output_get)
  (Variables|/Script/Engine.Texture|SetLODGroup "TEXTUREGROUP_Pixels2D" _output_get)
  ;[1 1 1] Threshold
  (bind _output_get_2 (LocalVariables|SetThresholdMaterial "/CorruptionGameplay/Materials/M_corruptionGradient_threshold.M_corruptionGradient_threshold"))
  (bind _returnvalue_3 (Rendering|Material|CreateDynamicMaterialInstance _output_get_2 "None" "None" :!class /Script/Engine.KismetMaterialLibrary))
  (Rendering|Material|SetTextureParameterValue _returnvalue_3 "GradientTexture" (Variables|Default|GetgradientTexture))
  (Rendering|Material|SetScalarParameterValue _returnvalue_3 "CorruptionTime" _returnvalue :!class /Script/Engine.MaterialInstanceDynamic)
  (Rendering|Material|SetScalarParameterValue _returnvalue_3 "Decrypt" (Math|Conversions|ToFloat(Boolean) (Variables|Generation|GetencryptGradient)) :!class /Script/Engine.MaterialInstanceDynamic)
  (Rendering|DrawMaterialtoRenderTarget (Variables|Default|GetRenderTarget_thresholded) _returnvalue_3)
  (LocalVariables|SetThresholdMaterial 0)
  ;[1 1 1] process
  (bind _output_get_3 (LocalVariables|SetProcessMaterial0 "/CorruptionGameplay/Materials/M_corruptionGradient_process.M_corruptionGradient_process"))
  (bind _returnvalue_4 (Rendering|Material|CreateDynamicMaterialInstance _output_get_3 "None" "None" :!class /Script/Engine.KismetMaterialLibrary))
  (Rendering|Material|SetTextureParameterValue _returnvalue_4 "StartingCorruption" (Variables|Default|GetstartingCorruptionTexture))
  (Rendering|Material|SetTextureParameterValue _returnvalue_4 "Thresholded" (Variables|Default|GetRenderTarget_thresholded))
  ;;[1 1 1] 2nd wave parameter - @bill.kladis
  (Rendering|Material|SetScalarParameterValue _returnvalue_4 "CubeTownCorruptionSpread" (LocalVariables|GetCubeTownCorruptionSpread) :!class /Script/Engine.MaterialInstanceDynamic)
  (Rendering|DrawMaterialtoRenderTarget (Variables|Default|GetRenderTarget_processed) _returnvalue_4)
  (LocalVariables|SetProcessMaterial0 0)
  ;[1 1 1] TODO: Is this the correct place?  THis needs to be called after all the textures have been generated.
  (CallFunction|CEInitializeCorruptionCoverageMap)
  ;[1 1 1] Flush the RVT to account for corruption changes
  (Development|ExecuteConsoleCommand "r.vt.Flush 1")
  ;[1 1 1] Cleanup once RT is created.
  (if (Variables|Default|Getcleanup)
    (Variables|Generation|SetInternalRenderTargetFootstep "")
    (Variables|Generation|SetInternalRenderTargetLayered "")
    (Variables|Generation|SetInternalRenderTargetLayeredCopy "")
    (Variables|Default|SetgradientTexture "")
    (Variables|Default|SetstartingCorruptionTexture "")
    (Variables|Default|SetRenderTarget_thresholded "")))

(fn ProcessStaticTextureFE
  (params (CorruptionTime :type Float%20%28double-precision%29) (CubeTownCorruptionSpread :type Float%20%28double-precision%29) (CorruptionSpread :type Corruption%20Spread%20Struct%20Structure))
  (Default|SetupLandscapeBounds(Landscape) 0)
  (Default|ProcessStaticTexture CorruptionSpread))

(fn ReportLandscapeBounds
  (bind _returnvalue (Rendering|Material|GetVectorParameterValue "/Game/Packages/Fortress_SharedMaterials/GlobalMaterialParameters/FortniteMaterialParameters.FortniteMaterialParameters" "CorruptionProjectionBias" :!class /Script/Engine.KismetMaterialLibrary))
  (bind _returnvalue_2 (Rendering|Material|GetVectorParameterValue "/Game/Packages/Fortress_SharedMaterials/GlobalMaterialParameters/FortniteMaterialParameters.FortniteMaterialParameters" "CorruptionProjectionScale" :!class /Script/Engine.KismetMaterialLibrary))
  (Development|PrintString (Utilities|String|Append "bias " (Utilities|String|ToString(LinearColor) _returnvalue) " scale " (Utilities|String|ToString(LinearColor) _returnvalue_2))))

(fn SetupLandscapeBounds
  (Default|SetupLandscapeBounds(Landscape) (Utilities|ResolveSoftReference (Variables|Landscape|Getlandscape))))

(fn SetupLandscapeBounds(Landscape)
  (params (Landscape :type Landscape%20Object%20Reference))
  (bind (_origin _boxextent) (Collision|GetActorBounds Landscape false false))
  (Utilities|IsValid Landscape
    (:Is%20Valid
      (Variables|Landscape|SetLandscapeBias (+ _origin _boxextent))
      (Variables|Landscape|SetLandscapeSize (* _boxextent 2.0)))
    (:Is%20Not%20Valid
      ;[1 1 1] apollo terrain values - used if no landscape plugged in
      (Variables|Landscape|SetLandscapeBias "135000.000000,135000.000000,135000.000000")
      (Variables|Landscape|SetLandscapeSize "270000.000000,270000.000000,270000.000000")))
  (Rendering|Material|SetVectorParameterValue "/Game/Packages/Fortress_SharedMaterials/GlobalMaterialParameters/FortniteMaterialParameters.FortniteMaterialParameters" "CorruptionProjectionBias" (Math|Conversions|ToLinearColor(Vector) (Variables|Landscape|GetLandscapeBias)) :!class /Script/Engine.KismetMaterialLibrary)
  (Rendering|Material|SetVectorParameterValue "/Game/Packages/Fortress_SharedMaterials/GlobalMaterialParameters/FortniteMaterialParameters.FortniteMaterialParameters" "CorruptionProjectionScale" (Math|Conversions|ToLinearColor(Vector) (Variables|Landscape|GetLandscapeSize)) :!class /Script/Engine.KismetMaterialLibrary))

(fn SortFootStepsUsingPython
  (bind _self self)
  (bind _returnvalue (Utilities|GetObjectPathString _self))
  (bind _returnvalue_2 (Utilities|String|Append "py Techart\\Corruption\\Sort.py " _returnvalue " Run"))
  (Development|PrintString _returnvalue_2)
  (Development|ExecuteConsoleCommand _returnvalue_2 (Game|GetPlayerController 0)))

(fn TestBeginPlay
  (bind _returnvalue (Game|GetPlayerController 0))
  (Development|PrintString "BP_CubeMovementGradient TestBeginPlay")
  (Utilities|IsValid _returnvalue
    (:Is%20Valid
      ;[1 1 1] in FE?
      (Utilities|Casting|CastToFortPlayerStateFrontEnd (Variables|/Script/Engine.Controller|GetPlayerState _returnvalue)
        (:then
          ;[1 1 1] FE Calendatr Events
          (Development|PrintString "BP_CubeMovementGradient FrontEndContext")
          (bind (_biseventactive _timesincebegin _timeuntilend _timespanratio) (CalendarEvents|GetCalendarEventInformation "LCCS01"))
          (if _biseventactive
            ;;[1 1 1] First Wave
            (Development|PrintString (Utilities|String|Append "Event LCCS01 Active " (Utilities|String|ToString(Float) _timespanratio)))
            (CallFunction|ProcessStaticTextureFE 0.0 0.0 (Utilities|Struct|MakeCorruptionSpreadStruct _timespanratio))
            (else
              (bind _biseventactive_2 (CalendarEvents|GetCalendarEventInformation "LCCSP01"))
              (if _biseventactive_2
                ;;[1 1 1] Pause
                (Development|PrintString "Event LCCSP01 Active")
                (CallFunction|ProcessStaticTextureFE 1.0 0.0 (Utilities|Struct|MakeCorruptionSpreadStruct 1.0))
                (else
                  (bind (_biseventactive_3 _timesincebegin_2 _timeuntilend_2 _timespanratio_2) (CalendarEvents|GetCalendarEventInformation "LCCS02"))
                  (if _biseventactive_3
                    ;;[1 1 1] CubeTownCorruptionSpread is active - this is the second wave of corruption
                    (Development|PrintString "Event LCCS02 Active 1.0")
                    (CallFunction|ProcessStaticTextureFE 1.0 0.0 (Utilities|Struct|MakeCorruptionSpreadStruct 1.0 _timespanratio_2))
                    (else
                      ;;[1 1 1] events aren't active .. 0 corruption.
                      (Development|PrintString "No LCC Events Active 0.001")
                      (CallFunction|ProcessStaticTextureFE 0.001 0.0 (Utilities|Struct|MakeCorruptionSpreadStruct 0.001 0.001)))))))))
        (:CastFailed
          _)))
    (:Is%20Not%20Valid
      _))
  ;[1 1 1] Gameplay Calendar
  (Development|PrintString "BP_CubeMovementGradient NOT FrontEndContext")
  (Development|PrintString "BP_CubeMovementGradient MatchReady")
  (Utilities|FlowControl|Sequence
    (:then_0
      (bind (_biseventactive_4 _timesincebegin_3 _timeuntilend_3 _timespanratio_3) (CalendarEvents|GetCalendarEventInformation "CSCM01")) ; entire duration of move1 == corruption duration\ndo not use CCS01, talk with tha; we're deleting that
      (if _biseventactive_4
        ;;[1 1 1] CSCM01 is active - this is the first wave of corruption.
        (Development|PrintString (Utilities|String|Append "CSCM01 Active " (Utilities|String|ToString(Float) _timespanratio_3)))
        (Variables|Default|SetcorruptionSpread (Utilities|Struct|MakeCorruptionSpreadStruct _timespanratio_3))
        (else
          (bind _biseventactive_5 (CalendarEvents|GetCalendarEventInformation "CCSP01"))
          (if _biseventactive_5
            ;;[1 1 1] CubeCorruptionSpreadPaused
            (Development|PrintString (Utilities|String|Append "CCSP01 Active "))
            (Variables|Default|SetcorruptionSpread (Utilities|Struct|MakeCorruptionSpreadStruct 1.0))
            (else
              ;;[1 1 1] CTC01 : CubeTownCorruptionSpread
              (bind (_biseventactive_6 _timesincebegin_4 _timeuntilend_4 _timespanratio_4) (CalendarEvents|GetCalendarEventInformation "CTCS01"))
              (if _biseventactive_6
                ;;[1 1 1] CubeTownCorruptionSpread is active - this is the second wave of corruption
                (Development|PrintString (Utilities|String|Append "CTCS01 Active " (Utilities|String|ToString(Float) _timespanratio_4)))
                (Variables|Default|SetcorruptionSpread (Utilities|Struct|MakeCorruptionSpreadStruct 1.0 _timespanratio_4))
                (else
                  ;;[1 1 1] events aren't active .. 0 corruption.
                  (Development|PrintString "No Events Active 0.00001")
                  (Variables|Default|SetcorruptionSpread (Utilities|Struct|MakeCorruptionSpreadStruct 0.001 0.001)))))))))
    (:then_1
      (if false
        (CallFunction|DebugStartOffsets)))))

(fn UpdateCorruptionState
  (bind _corruptioncoveragemap (Variables|CorruptionCoverageMap|GetCorruptionCoverageMap))
  (LocalVariables|SetNewIsAnyLocalPlayerInCorruption false)
  (for _array_element (Fortnite|GetLocalFortPlayerControllers true)
    (Utilities|IsValid _array_element
      (:Is%20Valid
        (Utilities|IsValid (Pawn|GetViewTarget _array_element)
          (:Is%20Valid
            (Utilities|IsValid _corruptioncoveragemap
              (:Is%20Valid
                (if (CorruptionCoverageMap|IsLocationCorrupted _corruptioncoveragemap (Transformation|GetActorLocation (Pawn|GetViewTarget _array_element)) (Variables|CorruptionCoverageMap|GetCorruptionCoveragePadding))
                  (LocalVariables|SetNewIsAnyLocalPlayerInCorruption true)))))))))
  (if (Variables|CorruptionCoverageMap|GetDebugCorruptionCoverageMap)
    (Development|PrintString (Utilities|String|Append "IsAnyLocalPlayerInCorruption: " (Utilities|String|ToString(Boolean) (Variables|CorruptionCoverageMap|GetIsAnyLocalPlayerInCorruption))) true false "(R=0.988433,G=0.020612,B=1.000000,A=1.000000)" (Variables|CorruptionCoverageMap|GetCorruptionStateUpdatePeriod)))
  (if (Math|Boolean|NotEqual(Boolean) (LocalVariables|GetNewIsAnyLocalPlayerInCorruption) (Variables|CorruptionCoverageMap|GetIsAnyLocalPlayerInCorruption))
    (bind _output_get (Variables|CorruptionCoverageMap|SetIsAnyLocalPlayerInCorruption (LocalVariables|GetNewIsAnyLocalPlayerInCorruption)))
    (CorruptionCoverageMap|CallOnIsAnyLocalPlayerinCorruptionChanged _output_get)
    (Fortnite|FortShippingLog self (Utilities|String|Append "Corruption Coverage: Local Player state changed to " (Utilities|String|SelectString "Inside" "Outside" _output_get)))
    (return)))

(fn ConstructionScript
  (bind _corruptionspread_cubecorruptionspread_7_62751f0f42c1cecd5f6b69b8e7717293 (Utilities|Struct|BreakCorruptionSpreadStruct (Variables|GetcorruptionSpread_CubeCorruptionSpread_7_62751F0F42C1CECD5F6B69B8E7717293)))
  (bind _corruptionspread_cubecorruptionspread_7_62751f0f42c1cecd5f6b69b8e7717293_2 (Utilities|Struct|BreakCorruptionSpreadStruct (Variables|GetcorruptionSpread_CubeCorruptionSpread_7_62751F0F42C1CECD5F6B69B8E7717293)))
  (if (Math|Boolean|NotEqual(Boolean) (Variables|Default|GetpreviewCubes) (Variables|Default|GetoldPreviewCubes))
    (Variables|Default|SetoldPreviewCubes (Variables|Default|GetpreviewCubes))
    (CallFunction|PreviewCubeInstances))
  (bind _returnvalue (Utilities|GetCommandLine))
  (if (Utilities|String|Contains _returnvalue "unattended")
    (Development|PrintString "skipping BP_CubeMovementGradient due to unattended used on command line")
    (else
      (Development|PrintString "BP_CubeMovementGradient")
      (if (!= (Variables|Debug|GetoldCorruptionTime) _corruptionspread_cubecorruptionspread_7_62751f0f42c1cecd5f6b69b8e7717293_2)
        (Variables|Debug|SetoldCorruptionTime _corruptionspread_cubecorruptionspread_7_62751f0f42c1cecd5f6b69b8e7717293)
        (if (Variables|Default|GetupdateRTs)
          (Utilities|IsValid (Variables|Default|GetgradientTexture)
            (:Is%20Valid
              (Utilities|IsValid (Variables|Default|GetRenderTarget_processed)
                (:Is%20Valid
                  (Default|ProcessStaticTexture (Variables|Default|GetcorruptionSpread))
                  (Default|SetupLandscapeBounds(Landscape) (Utilities|ResolveSoftReference (Variables|Landscape|Getlandscape))))))))))))
